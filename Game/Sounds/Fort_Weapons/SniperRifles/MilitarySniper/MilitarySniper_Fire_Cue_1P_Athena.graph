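# First-person fire sound cue for the Military Sniper rifle (Athena variant).
# Mixes three layers, each picking randomly from three waves: close shot, pitched-down close layer, and quieter shell casing delayed about half a second.
# Modulators vary pitch and volume on the layer and casing, and a root modulator slightly varies overall volume.
# Uses first-person weapon attenuation, sound class and gunfire concurrency assets.
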SoundCue /Game/Sounds/Fort_Weapons/SniperRifles/MilitarySniper/MilitarySniper_Fire_Cue_1P_Athena
root=Modulator_0
nodes [
 {id=Delay_0, children [Random_4], properties {delayMax=0.6, delayMin=0.5}}
 {id=Mixer_0, children [Modulator_1,Modulator_3,Modulator_2], properties {inputVolume [1,1.2,0.6]}}
 {id=Modulator_0, children [Mixer_0], properties {volumeMax=0.95, volumeMin=0.9}}
 {id=Modulator_1, children [Random_0], properties {volumeMax=1, volumeMin=1}}
 {id=Modulator_2, children [Delay_0], properties {pitchMax=1, pitchMin=0.8}}
 {id=Modulator_3, children [Random_1], properties {pitchMax=0.9, pitchMin=0.8, volumeMax=1.15, volumeMin=1.05}}
 {id=Random_0, children [WavePlayer_1,WavePlayer_2,WavePlayer_3], properties {bShouldExcludeFromBranchCulling=true, weights [1,1,1]}}
 {id=Random_1, children [WavePlayer_5,WavePlayer_6,WavePlayer_7], properties {bShouldExcludeFromBranchCulling=true, weights [1,1,1]}}
 {id=Random_4, children [WavePlayer_16, WavePlayer_17, WavePlayer_18], properties {bShouldExcludeFromBranchCulling=true, weights [1,1,1]}}
 {id=WavePlayer_1, properties {soundWaveAssetPtr=&/MilitarySniper_Fire_Close_01}}
 {id=WavePlayer_16, properties {soundWaveAssetPtr=&/MilitarySniper_Casing_01}}
 {id=WavePlayer_17, properties {soundWaveAssetPtr=&/MilitarySniper_Casing_02}}
 {id=WavePlayer_18, properties {soundWaveAssetPtr=&/MilitarySniper_Casing_03}}
 {id=WavePlayer_2, properties {soundWaveAssetPtr=&/MilitarySniper_Fire_Close_02}}
 {id=WavePlayer_3, properties {soundWaveAssetPtr=&/MilitarySniper_Fire_Close_03}}
 {id=WavePlayer_5, properties {soundWaveAssetPtr=&/MilitarySniper_Fire_CloseLayer_01}}
 {id=WavePlayer_6, properties {soundWaveAssetPtr=&/MilitarySniper_Fire_CloseLayer_02}}
 {id=WavePlayer_7, properties {soundWaveAssetPtr=&/MilitarySniper_Fire_CloseLayer_03}}
]
properties {
 attenuationSettings=&Fort_Weapon_Projectile_FirstPerson_Athena_Att
 maxDistance=2097152
 soundClassObject=&Fort_WeaponRanged_Fire_1P_PSM
 volumeMultiplier=1
 attenuationOverrides {distanceAlgorithm=NaturalSound, falloffDistance=4000, attenuationShapeExtents {x=1000}}
 concurrencyOverrides {resolutionRule=PreventNew}
 concurrencySet [&Weapon_Gunfire_FirstPerson_Conc]
}
