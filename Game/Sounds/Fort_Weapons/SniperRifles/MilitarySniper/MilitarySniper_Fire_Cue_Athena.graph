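# Sound cue for the Athena military sniper rifle gunshot, layered by close, medium and distant ranges.
# The close branch mixes random fire, pitched-down layer and delayed shell casing sounds under a third-person fire sound class.
# Medium and distant branches play random variants through their own attenuations, both under a distant fire sound class.
# All branches are mixed together, with local and global gunfire concurrency sets limiting playback.
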SoundCue /Game/Sounds/Fort_Weapons/SniperRifles/MilitarySniper/MilitarySniper_Fire_Cue_Athena
root=Mixer_15
nodes [
 {id=Attenuation_80, children [Modulator_0], properties {attenuationSettings=&Fort_Weapon_Projectile_Close_Athena_Att}}
 {
  id=Attenuation_81
  children [Modulator_13]
  properties {attenuationSettings=&Fort_Weapon_Projectile_Medium_Mono_Athena_Att}
 }
 {
  id=Attenuation_82
  children [Modulator_1]
  properties {attenuationSettings=&Fort_Weapon_Projectile_Distant_Mono_Athena_Att}
 }
 {id=Delay_0, children [Random_4], properties {delayMax=0.6, delayMin=0.5}}
 {id=Mixer_0, children [Random_0,Modulator_3,Modulator_2], properties {inputVolume [1,1.2,0.6]}}
 {id=Mixer_1, children [SoundClass_56, SoundClass_0], properties {inputVolume [0.5,0.5]}}
 {id=Mixer_15, children [SoundClass_55, Mixer_1], properties {inputVolume [0.7,1]}}
 {id=Modulator_0, children [Mixer_0], properties {volumeMax=0.95, volumeMin=0.9}}
 {id=Modulator_1, children [Random_6], properties {volumeMax=1.85, volumeMin=1.85}}
 {id=Modulator_13, children [Random_5], properties {volumeMax=2.15, volumeMin=2.05}}
 {id=Modulator_2, children [Delay_0], properties {pitchMax=1, pitchMin=0.8}}
 {id=Modulator_3, children [Random_1], properties {pitchMax=0.9, pitchMin=0.8, volumeMax=1.15, volumeMin=1.05}}
 {id=Random_0, children [WavePlayer_1,WavePlayer_2,WavePlayer_3], properties {bShouldExcludeFromBranchCulling=true, weights [1,1,1]}}
 {id=Random_1, children [WavePlayer_5,WavePlayer_6,WavePlayer_7], properties {bShouldExcludeFromBranchCulling=true, weights [1,1,1]}}
 {id=Random_4, children [WavePlayer_16, WavePlayer_17, WavePlayer_18], properties {bShouldExcludeFromBranchCulling=true, weights [1,1,1]}}
 {id=Random_5, children [WavePlayer_20, WavePlayer_21, WavePlayer_22], properties {bShouldExcludeFromBranchCulling=true, weights [1,1,1]}}
 {id=Random_6, children [WavePlayer_23, WavePlayer_24, WavePlayer_25], properties {bShouldExcludeFromBranchCulling=true, weights [1,1,1]}}
 {
  id=SoundClass_0
  children [Attenuation_82]
  properties {soundClassOverride=&Fort_WeaponRanged_Fire_Distant}
 }
 {
  id=SoundClass_55
  children [Attenuation_80]
  properties {soundClassOverride=&Fort_WeaponRanged_Fire_3P_PSM}
 }
 {
  id=SoundClass_56
  children [Attenuation_81]
  properties {soundClassOverride=&Fort_WeaponRanged_Fire_Distant}
 }
 {id=WavePlayer_1, properties {soundWaveAssetPtr=&/MilitarySniper_Fire_Close_01}}
 {id=WavePlayer_16, properties {soundWaveAssetPtr=&/MilitarySniper_Casing_01}}
 {id=WavePlayer_17, properties {soundWaveAssetPtr=&/MilitarySniper_Casing_02}}
 {id=WavePlayer_18, properties {soundWaveAssetPtr=&/MilitarySniper_Casing_03}}
 {id=WavePlayer_2, properties {soundWaveAssetPtr=&/MilitarySniper_Fire_Close_02}}
 {id=WavePlayer_20, properties {soundWaveAssetPtr=&/MilitarySniper_Fire_MediumNew_01}}
 {id=WavePlayer_21, properties {soundWaveAssetPtr=&/MilitarySniper_Fire_MediumNew_02}}
 {id=WavePlayer_22, properties {soundWaveAssetPtr=&/MilitarySniper_Fire_MediumNew_03}}
 {id=WavePlayer_23, properties {soundWaveAssetPtr=&/MilitarySniper_Fire_DistantNew_01}}
 {id=WavePlayer_24, properties {soundWaveAssetPtr=&/MilitarySniper_Fire_DistantNew_02}}
 {id=WavePlayer_25, properties {soundWaveAssetPtr=&/MilitarySniper_Fire_DistantNew_04}}
 {id=WavePlayer_3, properties {soundWaveAssetPtr=&/MilitarySniper_Fire_Close_03}}
 {id=WavePlayer_5, properties {soundWaveAssetPtr=&/MilitarySniper_Fire_CloseLayer_01}}
 {id=WavePlayer_6, properties {soundWaveAssetPtr=&/MilitarySniper_Fire_CloseLayer_02}}
 {id=WavePlayer_7, properties {soundWaveAssetPtr=&/MilitarySniper_Fire_CloseLayer_03}}
]
properties {
 maxDistance=27500
 volumeMultiplier=0.85
 attenuationOverrides {distanceAlgorithm=NaturalSound, falloffDistance=4000, attenuationShapeExtents {x=1000}}
 concurrencyOverrides {resolutionRule=PreventNew}
 concurrencySet [&Weapon_Gunfire_Local_Conc, &Weapon_Gunfire_Global_Conc]
}
