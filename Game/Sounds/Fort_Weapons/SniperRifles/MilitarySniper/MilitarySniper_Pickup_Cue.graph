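# Sound cue for picking up the Military Sniper rifle, playing in the loot-collect sound class with weapon-pickup attenuation.
# A single wave player feeds a modulator that boosts and slightly randomizes volume, offset by a lowered cue volume multiplier.
# Overridden concurrency allows one instance at a time and blocks new plays while one is active.
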SoundCue /Game/Sounds/Fort_Weapons/SniperRifles/MilitarySniper/MilitarySniper_Pickup_Cue
root=Modulator_1
nodes [
 {id=Modulator_1, children [WavePlayer_2], properties {volumeMax=2, volumeMin=1.8}}
 {id=WavePlayer_2, properties {soundWaveAssetPtr=&/MilitarySniper_Pickup_01}}
]
properties {
 attenuationSettings=&Fort_Weapon_Pickup_Att
 bOverrideConcurrency=true
 maxDistance=3500
 soundClassObject=&Fort_Loot_Collect
 volumeMultiplier=0.62
 concurrencyOverrides {maxCount=1, resolutionRule=PreventNew}
}
