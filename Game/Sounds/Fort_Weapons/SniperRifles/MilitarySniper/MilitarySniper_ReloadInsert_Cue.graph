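# Sound cue for the military sniper rifle's magazine-insert reload foley, using the weapon reload attenuation and WeaponFoley sound class.
# A mixer layers the main sniper insert sample with two quieter, delayed accents borrowed from auto rifle assets.
# The accents, a PDW insert slam and a pitched-down rifle bolt clip-in, start together shortly after the main sample.
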SoundCue /Game/Sounds/Fort_Weapons/SniperRifles/MilitarySniper/MilitarySniper_ReloadInsert_Cue
root=Mixer_0
nodes [
 {id=Delay_0, children [Modulator_1], properties {delayMax=0.4, delayMin=0.4}}
 {id=Delay_1, children [Modulator_0], properties {delayMax=0.4, delayMin=0.4}}
 {id=Mixer_0, children [WavePlayer_0,Delay_0,Delay_1], properties {inputVolume [1.1,0.5,0.5]}}
 {id=Modulator_0, children [WavePlayer_1], properties {pitchMax=0.8, pitchMin=0.8}}
 {id=Modulator_1, children [WavePlayer_2], properties {pitchMax=1, pitchMin=1}}
 {id=WavePlayer_0, properties {soundWaveAssetPtr=&/MilitarySniper_ReloadInsert_01}}
 {id=WavePlayer_1, properties {soundWaveAssetPtr=/Game/Sounds/Fort_Weapons/AutoRifle/Reload_Elements/rifle_bolt_single_clip_in}}
 {id=WavePlayer_2, properties {soundWaveAssetPtr=/Game/Sounds/Fort_Weapons/AutoRifle/PDW/PDW_Reload_InsertSlam_01}}
]
properties {
 attenuationSettings=&Fort_Weapon_Reload_Att
 bDebug=true
 bOverrideConcurrency=true
 maxDistance=2250
 soundClassObject=&Fort_WeaponFoley
 volumeMultiplier=0.501
}
