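# Sound cue for the start of the military sniper rifle reload, layering the reload foley into one timed sequence.
# A mixer plays the start sound immediately, a throw sound after a short delay and an impact after a longer one, each at its own volume.
# The impact is randomly picked from two waves and pitched down by a modulator.
# The cue uses weapon reload attenuation and the weapon foley sound class.
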SoundCue /Game/Sounds/Fort_Weapons/SniperRifles/MilitarySniper/MilitarySniper_ReloadStart_Cue
root=Mixer_0
nodes [
 {id=Delay_0, children [WavePlayer_2], properties {delayMax=0.5, delayMin=0.5}}
 {id=Delay_1, children [Random_1], properties {delayMax=1.1, delayMin=1.1}}
 {id=Mixer_0, children [WavePlayer_0,Delay_0,Modulator_0], properties {inputVolume [1,0.8,1.4]}}
 {id=Modulator_0, children [Delay_1], properties {pitchMax=0.9, pitchMin=0.8}}
 {id=Random_1, children [WavePlayer_3,WavePlayer_4], properties {weights [1,1]}}
 {id=WavePlayer_0, properties {soundWaveAssetPtr=&/MilitarySniper_ReloadStart_02}}
 {id=WavePlayer_2, properties {soundWaveAssetPtr=&/MilitarySniper_Reload_Throw_02}}
 {id=WavePlayer_3, properties {soundWaveAssetPtr=&/MilitarySniper_Reload_Impact_01}}
 {id=WavePlayer_4, properties {soundWaveAssetPtr=&/MilitarySniper_Reload_Impact_02}}
]
properties {
 attenuationSettings=&Fort_Weapon_Reload_Att
 bDebug=true
 bOverrideConcurrency=true
 maxDistance=2250
 soundClassObject=&Fort_WeaponFoley
 volumeMultiplier=0.601
}
