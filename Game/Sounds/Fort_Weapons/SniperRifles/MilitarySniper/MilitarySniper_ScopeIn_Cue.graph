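# Sound cue for the Military Sniper rifle's scope-in sound in the Fort weapon foley sound class.
# A single wave player feeds MilitarySniper_ScopeIn_01 straight to the output, with no randomization or modulation, at roughly half volume.
# It carries its own attenuation override values and overrides concurrency with a PreventNew rule, so new plays are refused once the limit is reached.
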SoundCue /Game/Sounds/Fort_Weapons/SniperRifles/MilitarySniper/MilitarySniper_ScopeIn_Cue
root=WavePlayer_3
nodes [{id=WavePlayer_3, properties {soundWaveAssetPtr=&/MilitarySniper_ScopeIn_01}}]
properties {
 bOverrideConcurrency=true
 maxDistance=2097152
 soundClassObject=&Fort_WeaponFoley
 volumeMultiplier=0.501
 attenuationOverrides {falloffDistance=1500, nonSpatializedRadiusStart=200, attenuationShapeExtents {x=1200}}
 concurrencyOverrides {resolutionRule=PreventNew}
}
