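# Sound cue for the scope-out foley of the military sniper rifle.
# A single wave player feeds the root at reduced volume, routed to the weapon foley sound class, with no randomization or mixing.
# A concurrency override blocks new instances at the limit rather than stopping old ones.
# Attenuation override values are stored, but their override flag is left off.
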SoundCue /Game/Sounds/Fort_Weapons/SniperRifles/MilitarySniper/MilitarySniper_ScopeOut_Cue
root=WavePlayer_0
nodes [{id=WavePlayer_0, properties {soundWaveAssetPtr=&/MilitarySniper_ScopeOut_01}}]
properties {
 bOverrideConcurrency=true
 maxDistance=2097152
 soundClassObject=&Fort_WeaponFoley
 volumeMultiplier=0.301
 attenuationOverrides {bSpatialize=false, falloffDistance=1500, nonSpatializedRadiusStart=200, attenuationShapeExtents {x=1200}}
 concurrencyOverrides {resolutionRule=PreventNew}
}
